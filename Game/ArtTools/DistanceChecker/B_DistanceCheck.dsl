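; B_DistanceCheck is an art-tool actor that lays out evenly spaced skeletal mesh markers on the surface below or above it, apparently as a distance reference.
; Its construction script loops over the copies count, stepping along the negative right vector by the distance variable times 100.
; Each step line-traces vertically across plus and minus traceDistance, and a hit adds two skeletal mesh components at the impact's Location.
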(event EventBeginPlay)

(event Collision|EventActorBeginOverlap
  (params (OtherActor :type Actor%20Object%20Reference)))

(event EventTick
  (params (DeltaSeconds :type Float%20%28single-precision%29)))

(fn ConstructionScript
  (for _index (range (Variables|Default|Getcopies))
    (bind (_outhit _returnvalue) (Collision|LineTraceByChannel (Math|Vector|MakeVector (.x (+ (Transformation|GetActorLocation) (* (Transformation|GetActorRightVector) (* (+ _index 1) (* (Variables|Default|Getdistance) -1e+02))))) (.y (+ (Transformation|GetActorLocation) (* (Transformation|GetActorRightVector) (* (+ _index 1) (* (Variables|Default|Getdistance) -1e+02))))) (+ (.z (+ (Transformation|GetActorLocation) (* (Transformation|GetActorRightVector) (* (+ _index 1) (* (Variables|Default|Getdistance) -1e+02))))) (Variables|Default|GettraceDistance))) (Math|Vector|MakeVector (.x (+ (Transformation|GetActorLocation) (* (Transformation|GetActorRightVector) (* (+ _index 1) (* (Variables|Default|Getdistance) -1e+02))))) (.y (+ (Transformation|GetActorLocation) (* (Transformation|GetActorRightVector) (* (+ _index 1) (* (Variables|Default|Getdistance) -1e+02))))) (+ (.z (+ (Transformation|GetActorLocation) (* (Transformation|GetActorRightVector) (* (+ _index 1) (* (Variables|Default|Getdistance) -1e+02))))) (* (Variables|Default|GettraceDistance) -1.0)))))
    (if _returnvalue
      (bind _returnvalue_2 (AddComponent|NODE_AddSkeletalMeshComponent-0))
      (Transformation|SetWorldLocation _returnvalue_2 (.Location (Collision|BreakHitResult _outhit)) false false)
      (bind _returnvalue_3 (AddComponent|NODE_AddSkeletalMeshComponent-1))
      (Transformation|SetWorldLocation _returnvalue_3 (.Location (Collision|BreakHitResult _outhit)) false false))))
